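# Material function for the Fancy Candy character that fakes facial animation by swapping pre-modeled eye and eyebrow shapes through world position offset.
# Squint, blink, brow and glabella pose parameters are thresholded by scaling and rounding, then lerp-select vertex color values for the current eye (red) and brow (green) shapes.
# VertexColorMask results are combined with a head green-value mask.
# Vertices outside the combined mask collapse to the object position, added to incoming WPO.
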MaterialFunction /BRCosmetics/Characters/Player/Male/Medium/Bodies/M_MED_Fancy_Candy/Materials/MF_FancyCandyCheapEyes
graph {nodes [
  n1 = VertexColor()
  ChannelMaskParameter(None=@n1) {group=Banana, parameterName=EyeVertexColorChannel Red}
  n2 = GetMaterialAttributes(None=FunctionInput() {inputType=FunctionInput_MaterialAttributes}) {attributeGetTypes [#1]}
  n3 = WorldPosition()
  n4 = ObjectPositionWS()
  n5 = VertexColor()
  n6 = Constant() {r=1}
  n7 = ScalarParameter() {defaultValue=0.3, group=Banana, parameterName="Eye (Default)"}
  n8 = ScalarParameter() {defaultValue=0.5, group=Banana, parameterName=Eye Squint}
  n9 = Max(A=ScalarParameter() {group=Banana_Poses, parameterName=R_squint_inner_pose}) {constB=0}
  n10 = Divide(B=ScalarParameter() {defaultValue=0.1, parameterName=TH_squint}) {constA=0.5}
  n11 = Clamp(None=Round(None=Multiply(A=@n9, B=@n10)))
  n12 = Multiply(
   A=ScalarParameter() {group=Banana, parameterName=C_glabella_up_pose}
   B=Divide(B=ScalarParameter() {defaultValue=2, parameterName=TH_glabella_up}) {constA=0.5}
  )
  n13 = OneMinus(None=Clamp(None=Round(None=@n12)))
  n14 = Multiply(
   A=ScalarParameter() {group=Banana, parameterName=R_brow_up_pose}
   B=Divide(B=ScalarParameter() {defaultValue=0.3, parameterName=TH_brow_up}) {constA=0.5}
  )
  n15 = Multiply(
   A=ScalarParameter() {group=Banana_Poses, parameterName=C_glabella_down_pose}
   B=Divide(B=ScalarParameter() {defaultValue=0.2, parameterName=TH_glabella_down}) {constA=0.5}
  )
  n16 = Max(A=Clamp(None=Round(None=@n14)), B=Clamp(None=Round(None=@n15)))
  n17 = Multiply(
   A=Max(A=ScalarParameter() {group=Banana_Poses, parameterName=R_blink_pose}) {constB=0}
   B=Divide(B=ScalarParameter() {defaultValue=0.3, parameterName=TH_narrow}) {constA=0.5}
  )
  n18 = LinearInterpolate(
   A=LinearInterpolate(
    A=LinearInterpolate(A=@n7, B=@n8, Alpha=@n11)
    B=ScalarParameter() {defaultValue=0.3, group=Banana, parameterName="Eye (Default)"}
    Alpha=Min(A=@n13, B=@n16)
   )
   B=ScalarParameter() {defaultValue=0.6, group=Banana, parameterName=Eye Narrow}
   Alpha=Clamp(None=Round(None=@n17))
  )
  n19 = Constant() {r=1}
  n20 = ScalarParameter() {defaultValue=0.2, group=Banana, parameterName="Glabella Down (Default)"}
  n21 = ScalarParameter() {defaultValue=0.05, group=Banana, parameterName=Brow Up}
  n22 = Max(A=ScalarParameter() {group=Banana, parameterName=R_brow_up_pose}) {constB=0}
  n23 = ScalarParameter() {defaultValue=0.1, group=Banana, parameterName=TH_Brow}
  n24 = Divide(B=@n23) {constA=0.5}
  n25 = Clamp(None=Round(None=Multiply(A=@n22, B=@n24)))
  n26 = ScalarParameter() {group=Banana, parameterName=C_glabella_down_pose}
  n27 = Round(None=Multiply(A=Max(A=@n26) {constB=0}, B=Divide(B=@n23) {constA=0.5}))
  n28 = Multiply(A=Max(A=ScalarParameter() {group=Banana, parameterName=C_glabella_up_pose}) {constB=0}, B=Divide(B=@n23) {constA=0.5})
  n29 = LinearInterpolate(
   A=LinearInterpolate(
    A=LinearInterpolate(A=@n20, B=@n21, Alpha=@n25)
    B=ScalarParameter() {defaultValue=0.2, group=Banana, parameterName="Glabella Down (Default)"}
    Alpha=Clamp(None=@n27)
   )
   B=ScalarParameter() {defaultValue=0.1, group=Banana, parameterName=Glabella Up}
   Alpha=Clamp(None=Round(None=@n28))
  )
  n30 = Max(
   A=MaterialFunctionCall!(
    Hide Element Toggle=Constant() {r=1}
    VertexColorChannel=ChannelMaskParameter(None=@n5) {group=Banana, maskChannel=Green, parameterName=EyeBrowVertexColorChannel Green}.R
    Vertex Value to hide=ScalarParameter() {defaultValue=1, parameterName=HeadGreenValue}
   ) {materialFunction=/Game/Packages/Fortress_SharedMaterials/Base_Material_Functions/VertexColorMask}.Result
   B=Max(
    A=MaterialFunctionCall(Hide Element Toggle=@n6, VertexColorChannel=@n1.R, Vertex Value to hide=@n18) {materialFunction=&VertexColorMask}.Result
    B=MaterialFunctionCall(Hide Element Toggle=@n19, VertexColorChannel=@n5.G, Vertex Value to hide=@n29) {materialFunction=&VertexColorMask}.Result
   )
  )
  n31 = Multiply(A=Multiply(A=Subtract(A=@n3.XYZ, B=@n4)) {constB=-1}, B=Saturate(None=OneMinus(None=@n30)))
  FunctionOutput(None=SetMaterialAttributes!(MaterialAttributes=@n2.MaterialAttributes, WorldPositionOffset=Add(A=@n2.WorldPositionOffset, B=@n31)) {attributeSetTypes [#1]})
 ]}
