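# Material function for the Fancy Candy character that fakes mouth animation by switching mustache geometry from facial pose parameters.
# Smile and jaw-open values are thresholded and rounded to pick an up, down or neutral vertex value, with jaw-open overriding smile.
# VertexColorMask compares it with green vertex color, and the inverted result collapses vertices toward the object position.
# That offset is added to the incoming attributes' World Position Offset.
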MaterialFunction /BRCosmetics/Characters/Player/Male/Medium/Bodies/M_MED_Fancy_Candy/Materials/MF_FancyCandyCheapMouth
graph {nodes [
  n1 = VertexColor()
  ChannelMaskParameter(None=@n1) {group=Banana, maskChannel=Green, parameterName=EyeBrowVertexColorChannel Green, defaultValue=<rgba 0 1 0 0>}
  n2 = GetMaterialAttributes(None=FunctionInput() {inputType=FunctionInput_MaterialAttributes}) {attributeGetTypes [#1]}
  n3 = WorldPosition()
  n4 = ObjectPositionWS()
  n5 = ScalarParameter() {defaultValue=0.2, group=Banana, parameterName="Mustache Neutral (Default)"}
  n6 = ScalarParameter() {defaultValue=0.05, group=Banana, parameterName=Mustache Up}
  n7 = Max(A=ScalarParameter() {group=Banana, parameterName=R_smile_pose}) {constB=0}
  n8 = ScalarParameter() {defaultValue=0.1, group=Banana, parameterName=THreshold}
  n9 = Divide(B=@n8) {constA=0.5}
  n10 = Clamp(None=Round(None=Multiply(A=@n7, B=@n9)))
  n11 = ScalarParameter() {group=Banana, parameterName=jaw_open_pose}
  n12 = Round(None=Multiply(A=Max(A=@n11) {constB=0}, B=Divide(B=@n8) {constA=0.5}))
  n13 = MaterialFunctionCall!(
   Hide Element Toggle=Constant() {r=1}
   VertexColorChannel=@n1.G
   Vertex Value to hide=LinearInterpolate(
    A=LinearInterpolate(A=@n5, B=@n6, Alpha=@n10)
    B=ScalarParameter() {defaultValue=0.1, group=Banana, parameterName=Mustache Down}
    Alpha=Clamp(None=@n12)
   )
  ) {materialFunction=/Game/Packages/Fortress_SharedMaterials/Base_Material_Functions/VertexColorMask}
  n14 = Multiply(A=Multiply(A=Subtract(A=@n3.XYZ, B=@n4)) {constB=-1}, B=Saturate(None=OneMinus(None=@n13.Result)))
  FunctionOutput(None=SetMaterialAttributes!(MaterialAttributes=@n2.MaterialAttributes, WorldPositionOffset=Add(A=@n2.WorldPositionOffset, B=@n14)) {attributeSetTypes [#1]})
 ]}
