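# Physics asset for the right-arm bandana on the M_MED_Fancy_Candy male medium character body, covering its dynamic bandana bones.
# Box bodies cover two separate four-bone chains, upper and lower, each running from a master bone through three segments.
# Each segment is constrained to the previous one with tight, uniform swing limits and an even narrower twist limit, so the cloth bends only slightly per joint.
# No constraint links the two chains.
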PhysicsAsset /BRCosmetics/Characters/Player/Male/Medium/Bodies/M_MED_Fancy_Candy/Meshes/M_MED_Fancy_Candy_Physics_ArmRight
bodies {
 dyn_bandanaLower_01 [{shapeType=Box, extentX=3.222, extentY=4.5412, extentZ=1.4613, center=<4.8319 0.0675 -0.0615>}]
 dyn_bandanaLower_02 [{shapeType=Box, extentX=4.8961, extentY=5.0065, extentZ=2.1216, center=<3.9568 0.0362 0.4198>}]
 dyn_bandanaLower_03 [{shapeType=Box, extentX=6.1741, extentY=3.6127, extentZ=1.5856, center=<3.0139 -0.2731 -0.4557>}]
 dyn_bandanaLower_mstr [{shapeType=Box, extentX=1.6636, extentY=3.2618, extentZ=1.752, center=<0.3641 0.0493 0.1014>}]
 dyn_bandanaUpper_01 [{shapeType=Box, extentX=3.3564, extentY=4.2327, extentZ=1.6444, center=<3.9001 0.0618 -0.0344>}]
 dyn_bandanaUpper_02 [{shapeType=Box, extentX=4.0808, extentY=4.8077, extentZ=2.3015, center=<3.1383 0.017 0.1055>}]
 dyn_bandanaUpper_03 [{shapeType=Box, extentX=3.74, extentY=2.4706, extentZ=1.2921, center=<2.8863 0.2769 0.0799>}]
 dyn_bandanaUpper_mstr [{shapeType=Box, extentX=1.9443, extentY=2.6643, extentZ=2.1387, center=<0.2582 -0.0905 0.7064>}]
}
constraints [
 {bone1Name=dyn_bandanaLower_01, bone2Name=dyn_bandanaLower_mstr, swing1LimitDegrees=10, swing2LimitDegrees=10, twistLimitDegrees=2}
 {bone1Name=dyn_bandanaLower_02, bone2Name=dyn_bandanaLower_01, swing1LimitDegrees=10, swing2LimitDegrees=10, twistLimitDegrees=2}
 {bone1Name=dyn_bandanaLower_03, bone2Name=dyn_bandanaLower_02, swing1LimitDegrees=10, swing2LimitDegrees=10, twistLimitDegrees=2}
 {bone1Name=dyn_bandanaUpper_01, bone2Name=dyn_bandanaUpper_mstr, swing1LimitDegrees=10, swing2LimitDegrees=10, twistLimitDegrees=2}
 {bone1Name=dyn_bandanaUpper_02, bone2Name=dyn_bandanaUpper_01, swing1LimitDegrees=10, swing2LimitDegrees=10, twistLimitDegrees=2}
 {bone1Name=dyn_bandanaUpper_03, bone2Name=dyn_bandanaUpper_02, swing1LimitDegrees=10, swing2LimitDegrees=10, twistLimitDegrees=2}
]
