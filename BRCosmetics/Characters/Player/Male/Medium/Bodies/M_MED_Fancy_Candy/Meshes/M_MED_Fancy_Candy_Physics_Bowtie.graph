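# Physics asset for the bowtie on the M_MED_Fancy_Candy medium male player body, defining simulation bodies for its dynamic bones.
# A small capsule on the master bone links to a box on the center bowtie bone through a joint with tight swing and twist limits.
# Left and right wing boxes hang off the center with narrow limits and one swing axis set to zero, allowing only slight wobble.
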PhysicsAsset /BRCosmetics/Characters/Player/Male/Medium/Bodies/M_MED_Fancy_Candy/Meshes/M_MED_Fancy_Candy_Physics_Bowtie
bodies {
 dyn_bowtie [{shapeType=Box, extentX=5, extentY=5, extentZ=5}]
 dyn_bowtie_01_l [{shapeType=Box, extentX=10.1347, extentY=4.7629, extentZ=13.9033, center=<5.499 0.5735 0.1003>}]
 dyn_bowtie_01_r [{shapeType=Box, extentX=9.2688, extentY=4.4994, extentZ=13.9521, center=<5.3836 -0.4661 -0.0281>}]
 dyn_bowtie_mstr [{shapeType=Capsule, length=0.505, radius=0.505}]
}
constraints [
 {bone1Name=dyn_bowtie, bone2Name=dyn_bowtie_mstr, swing1LimitDegrees=2, swing2LimitDegrees=2, twistLimitDegrees=2}
 {bone1Name=dyn_bowtie_01_r, bone2Name=dyn_bowtie, swing1LimitDegrees=5, swing2LimitDegrees=0, twistLimitDegrees=1}
 {bone1Name=dyn_bowtie_01_l, bone2Name=dyn_bowtie, swing1LimitDegrees=5, swing2LimitDegrees=0, twistLimitDegrees=1}
]
