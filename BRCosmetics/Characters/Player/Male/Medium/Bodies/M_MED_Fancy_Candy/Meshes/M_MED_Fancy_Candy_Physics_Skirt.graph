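# Physics asset for the skirt on the Fancy Candy medium male player cosmetic body, covering only the skirt and upper-leg bones.
# Front, side and back chains on each side run from a tiny master box through two flat panel boxes.
# Each master joins its same-side thigh-twist capsule with free angular motion, while the panel joints below have tight swing and twist limits.
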PhysicsAsset /BRCosmetics/Characters/Player/Male/Medium/Bodies/M_MED_Fancy_Candy/Meshes/M_MED_Fancy_Candy_Physics_Skirt
bodies {
 dyn_skirtBack_01_l [{shapeType=Box, extentX=10, extentY=6, extentZ=1, center=<3 0 0>, rotation=<rot 0 -90 0>}]
 dyn_skirtBack_01_r [{shapeType=Box, extentX=10, extentY=6, extentZ=1, center=<3 0 0>, rotation=<rot 0 -90 0>}]
 dyn_skirtBack_02_l [{shapeType=Box, extentX=10, extentY=6, extentZ=1, center=<3 0 0>, rotation=<rot 0 -90 0>}]
 dyn_skirtBack_02_r [{shapeType=Box, extentX=10, extentY=6, extentZ=1, center=<3 0 0>, rotation=<rot 0 -90 0>}]
 dyn_skirtBack_mstr_l [{shapeType=Box, extentX=1.01, extentY=1.01, extentZ=1.01, rotation=<rot 0 -90 0>}]
 dyn_skirtBack_mstr_r [{shapeType=Box, extentX=1.01, extentY=1.01, extentZ=1.01, rotation=<rot 0 -90 0>}]
 dyn_skirtFront_01_l [{shapeType=Box, extentX=10, extentY=6, extentZ=1, center=<3 0 0>, rotation=<rot 0 -90 0>}]
 dyn_skirtFront_01_r [{shapeType=Box, extentX=10, extentY=6, extentZ=1, center=<3 0 0>, rotation=<rot 0 -90 0>}]
 dyn_skirtFront_02_l [{shapeType=Box, extentX=10, extentY=6, extentZ=1, center=<3 0 0>, rotation=<rot 0 -90 0>}]
 dyn_skirtFront_02_r [{shapeType=Box, extentX=10, extentY=6, extentZ=1, center=<3 0 0>, rotation=<rot 0 -90 0>}]
 dyn_skirtFront_mstr_l [{shapeType=Box, extentX=1.01, extentY=1.01, extentZ=1.01, rotation=<rot 0 -90 0>}]
 dyn_skirtFront_mstr_r [{shapeType=Box, extentX=1.01, extentY=1.01, extentZ=1.01, rotation=<rot 0 -90 0>}]
 dyn_skirtSide_01_l [{shapeType=Box, extentX=10, extentY=6, extentZ=1, center=<3 0 0>, rotation=<rot 0 -90 0>}]
 dyn_skirtSide_01_r [{shapeType=Box, extentX=10, extentY=6, extentZ=1, center=<3 0 0>, rotation=<rot 0 -90 0>}]
 dyn_skirtSide_02_l [{shapeType=Box, extentX=10, extentY=6, extentZ=1, center=<3 0 0>, rotation=<rot 0 -90 0>}]
 dyn_skirtSide_02_r [{shapeType=Box, extentX=10, extentY=6, extentZ=1, center=<3 0 0>, rotation=<rot 0 -90 0>}]
 dyn_skirtSide_mstr_l [{shapeType=Box, extentX=1.01, extentY=1.01, extentZ=1.01, rotation=<rot 0 -90 0>}]
 dyn_skirtSide_mstr_r [{shapeType=Box, extentX=1.01, extentY=1.01, extentZ=1.01, rotation=<rot 0 -90 0>}]
 thigh_twist_01_l [{shapeType=Capsule, length=19.3902, radius=7, rotation=<rot 90 0 0>}]
 thigh_twist_01_r [{shapeType=Capsule, length=19.4495, radius=7, rotation=<rot 90 0 0>}]
}
constraints [
 {
  bone1Name=dyn_skirtFront_mstr_l
  bone2Name=thigh_twist_01_l
  swing1LimitDegrees=45
  swing1Motion=Free
  swing2LimitDegrees=45
  swing2Motion=Free
  twistLimitDegrees=45
  twistMotion=Free
 }
 {bone1Name=dyn_skirtFront_01_l, bone2Name=dyn_skirtFront_mstr_l, swing1LimitDegrees=15, swing2LimitDegrees=25, twistLimitDegrees=5}
 {bone1Name=dyn_skirtFront_02_l, bone2Name=dyn_skirtFront_01_l, swing1LimitDegrees=15, swing2LimitDegrees=25, twistLimitDegrees=5}
 {
  bone1Name=dyn_skirtSide_mstr_l
  bone2Name=thigh_twist_01_l
  swing1LimitDegrees=45
  swing1Motion=Free
  swing2LimitDegrees=45
  swing2Motion=Free
  twistLimitDegrees=45
  twistMotion=Free
 }
 {bone1Name=dyn_skirtSide_01_l, bone2Name=dyn_skirtSide_mstr_l, swing1LimitDegrees=15, swing2LimitDegrees=25, twistLimitDegrees=5}
 {bone1Name=dyn_skirtSide_02_l, bone2Name=dyn_skirtSide_01_l, swing1LimitDegrees=15, swing2LimitDegrees=25, twistLimitDegrees=5}
 {
  bone1Name=dyn_skirtBack_mstr_l
  bone2Name=thigh_twist_01_l
  swing1LimitDegrees=45
  swing1Motion=Free
  swing2LimitDegrees=45
  swing2Motion=Free
  twistLimitDegrees=45
  twistMotion=Free
 }
 {bone1Name=dyn_skirtBack_01_l, bone2Name=dyn_skirtBack_mstr_l, swing1LimitDegrees=15, swing2LimitDegrees=25, twistLimitDegrees=5}
 {bone1Name=dyn_skirtBack_02_l, bone2Name=dyn_skirtBack_01_l, swing1LimitDegrees=15, swing2LimitDegrees=25, twistLimitDegrees=5}
 {
  bone1Name=dyn_skirtFront_mstr_r
  bone2Name=thigh_twist_01_r
  swing1LimitDegrees=45
  swing1Motion=Free
  swing2LimitDegrees=45
  swing2Motion=Free
  twistLimitDegrees=45
  twistMotion=Free
 }
 {bone1Name=dyn_skirtFront_01_r, bone2Name=dyn_skirtFront_mstr_r, swing1LimitDegrees=15, swing2LimitDegrees=25, twistLimitDegrees=5}
 {bone1Name=dyn_skirtFront_02_r, bone2Name=dyn_skirtFront_01_r, swing1LimitDegrees=15, swing2LimitDegrees=25, twistLimitDegrees=5}
 {
  bone1Name=dyn_skirtSide_mstr_r
  bone2Name=thigh_twist_01_r
  swing1LimitDegrees=45
  swing1Motion=Free
  swing2LimitDegrees=45
  swing2Motion=Free
  twistLimitDegrees=45
  twistMotion=Free
 }
 {bone1Name=dyn_skirtSide_01_r, bone2Name=dyn_skirtSide_mstr_r, swing1LimitDegrees=15, swing2LimitDegrees=25, twistLimitDegrees=5}
 {bone1Name=dyn_skirtSide_02_r, bone2Name=dyn_skirtSide_01_r, swing1LimitDegrees=15, swing2LimitDegrees=25, twistLimitDegrees=5}
 {
  bone1Name=dyn_skirtBack_mstr_r
  bone2Name=thigh_twist_01_r
  swing1LimitDegrees=45
  swing1Motion=Free
  swing2LimitDegrees=45
  swing2Motion=Free
  twistLimitDegrees=45
  twistMotion=Free
 }
 {bone1Name=dyn_skirtBack_01_r, bone2Name=dyn_skirtBack_mstr_r, swing1LimitDegrees=15, swing2LimitDegrees=25, twistLimitDegrees=5}
 {bone1Name=dyn_skirtBack_02_r, bone2Name=dyn_skirtBack_01_r, swing1LimitDegrees=15, swing2LimitDegrees=25, twistLimitDegrees=5}
]
