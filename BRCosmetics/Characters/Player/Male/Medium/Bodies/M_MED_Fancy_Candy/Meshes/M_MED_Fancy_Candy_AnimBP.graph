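# Animation Blueprint for the Fancy Candy medium male body cosmetic, copying the attached parent mesh's pose and adding secondary motion to arm, bowtie and skirt.
# Each update converts bIsCrouching to a float alpha that, with interpolation, eases in additive offsets on six skirt master bones.
# Chained RigidBody nodes then simulate right arm, bowtie and skirt in base-bone space.
# The skirt switches between standing and crouching physics assets via bIsCrouching.
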AnimBlueprint /BRCosmetics/Characters/Player/Male/Medium/Bodies/M_MED_Fancy_Candy/Meshes/M_MED_Fancy_Candy_AnimBP
target_skeleton=&/M_MED_Fancy_Candy_Skeleton
graphs [
 AnimGraph {nodes [
   n1 = Get crouching() -> (crouching:double)
   n2 = "Transform (Modify) Bone - Bone: dyn_skirtFront_mstr_l"(
    ComponentPose=Local To Component(LocalPose=Copy Pose From Mesh() {bUseAttachedParent=true})
    Alpha=@n1
    Translation="0,-13,0"
    Rotation="0,0,0"
    Scale="1,1,1"
   ) {
    translationMode=BMM_Additive
    alphaBoolBlend {blendOption=Linear}
    alphaScaleBiasClamp {bInterpResult=true, interpSpeedDecreasing=5, interpSpeedIncreasing=5}
    boneToModify {boneName=dyn_skirtFront_mstr_l}
    translation {y=-13}
   }
   n3 = "Transform (Modify) Bone - Bone: dyn_skirtSide_mstr_l"(ComponentPose=@n2, Alpha=@n1, Translation="0,-16,0", Rotation="0,0,0", Scale="1,1,1") {
    translationMode=BMM_Additive
    alphaBoolBlend {blendOption=Linear}
    alphaScaleBiasClamp {bInterpResult=true, interpSpeedDecreasing=5, interpSpeedIncreasing=5}
    boneToModify {boneName=dyn_skirtSide_mstr_l}
    translation {y=-16}
   }
   n4 = "Transform (Modify) Bone - Bone: dyn_skirtBack_mstr_l"(ComponentPose=@n3, Alpha=@n1, Translation="0,-20,0", Rotation="0,0,0", Scale="1,1,1") {
    translationMode=BMM_Additive
    alphaBoolBlend {blendOption=Linear}
    alphaScaleBiasClamp {bInterpResult=true, interpSpeedDecreasing=5, interpSpeedIncreasing=5}
    boneToModify {boneName=dyn_skirtBack_mstr_l}
    translation {y=-20}
   }
   n5 = "Transform (Modify) Bone - Bone: dyn_skirtFront_mstr_r"(ComponentPose=@n4, Alpha=@n1, Translation="0,0,5", Rotation="0,0,0", Scale="1,1,1") {
    translationMode=BMM_Additive
    alphaBoolBlend {blendOption=Linear}
    alphaScaleBiasClamp {bInterpResult=true, interpSpeedDecreasing=5, interpSpeedIncreasing=5}
    boneToModify {boneName=dyn_skirtFront_mstr_r}
    translation {z=5}
   }
   n6 = "Transform (Modify) Bone - Bone: dyn_skirtSide_mstr_r"(ComponentPose=@n5, Alpha=@n1, Translation="0.000,0,10", Rotation="0,0,0", Scale="1,1,1") {
    translationMode=BMM_Additive
    alphaBoolBlend {blendOption=Linear}
    alphaScaleBiasClamp {bInterpResult=true, interpSpeedDecreasing=5, interpSpeedIncreasing=5}
    boneToModify {boneName=dyn_skirtSide_mstr_r}
    translation {z=10}
   }
   n7 = "Transform (Modify) Bone - Bone: dyn_skirtBack_mstr_r"(ComponentPose=@n6, Alpha=@n1, Translation="0,0,20", Rotation="0,0,0", Scale="1,1,1") {
    translationMode=BMM_Additive
    alphaBoolBlend {blendOption=Linear}
    alphaScaleBiasClamp {bInterpResult=true, interpSpeedDecreasing=5, interpSpeedIncreasing=5}
    boneToModify {boneName=dyn_skirtBack_mstr_r}
    translation {z=20}
   }
   n8 = RigidBody(ComponentPose=@n7) {
    bClampLinearTranslationLimitToRefPose=true
    bOverrideWorldGravity=true
    overridePhysicsAsset=&/M_MED_Fancy_Candy_Physics_ArmRight
    simulationSpace=BaseBoneSpace
    alphaBoolBlend {blendOption=Linear}
    baseBoneRef {boneName=upperarm_r}
    componentAppliedLinearAccClamp=<5000 5000 5000>
    componentLinearAccScale=<0.3 0.3 0.3>
    componentLinearVelScale=<0.3 0.3 0.3>
    overrideWorldGravity {z=-980}
   }
   n9 = RigidBody(ComponentPose=@n8) {
    bClampLinearTranslationLimitToRefPose=true
    bOverrideWorldGravity=true
    overridePhysicsAsset=&/M_MED_Fancy_Candy_Physics_Bowtie
    simulationSpace=BaseBoneSpace
    alphaBoolBlend {blendOption=Linear}
    baseBoneRef {boneName=spine_04}
    componentAppliedLinearAccClamp=<5000 5000 5000>
    componentLinearAccScale=<0.3 0.3 0.3>
    componentLinearVelScale=<0.3 0.3 0.3>
    overrideWorldGravity {z=-980}
   }
   n10 = Get bIsCrouching()
   n11 = RigidBody!(ComponentPose=@n9, bAlphaBoolEnabled:bool=NOT Boolean(A=@n10), ExternalForce) -> (Pose) {
    alphaInputType=Bool
    bClampLinearTranslationLimitToRefPose=true
    bOverrideWorldGravity=true
    overridePhysicsAsset=&/M_MED_Fancy_Candy_Physics_Skirt
    simulationSpace=BaseBoneSpace
    alphaBoolBlend {blendOption=Linear}
    baseBoneRef {boneName=pelvis}
    componentAppliedLinearAccClamp=<5000 5000 5000>
    componentLinearAccScale=<0.3 0.3 0.3>
    componentLinearVelScale=<0.3 0.3 0.3>
    overrideWorldGravity {z=-980}
   }
   n12 = RigidBody(ComponentPose=@n11, bAlphaBoolEnabled=@n10) {
    alphaInputType=Bool
    bClampLinearTranslationLimitToRefPose=true
    bOverrideWorldGravity=true
    overridePhysicsAsset=&/M_MED_Fancy_Candy_Physics_SkirtCrouching
    simulationSpace=BaseBoneSpace
    alphaBoolBlend {blendOption=Linear}
    baseBoneRef {boneName=pelvis}
    componentAppliedLinearAccClamp=<5000 5000 5000>
    componentLinearAccScale=<0.3 0.3 0.3>
    componentLinearVelScale=<0.3 0.3 0.3>
    overrideWorldGravity {z=-980}
   }
   Output Pose(Result=Component To Local(ComponentPose=@n12)) {layerGroup=None}
  ]}
 EventGraph {nodes [
   Event BlueprintUpdateAnimation()
   > Set crouching(crouching:double="To Float (Boolean)"(InBool=Get bIsCrouching())) -> (Output_Get:double)
  ]}
]
variables [crouching {type=number}]
