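# Physics asset for the Fancy Candy medium male body cosmetic, covering a chain of head cane dynamic bones.
# Five capsule bodies of equal radius and decreasing length follow the cane bones outward from a tiny capsule on the master bone.
# Constraints link each segment to the one before it, with tight, equal swing and twist limits that allow only slight flex at each joint.
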PhysicsAsset /BRCosmetics/Characters/Player/Male/Medium/Bodies/M_MED_Fancy_Candy/Meshes/Parts/M_MED_Fancy_Candy_Physics
bodies {
 dyn_headCane_01 [{shapeType=Capsule, length=12.6151, radius=10, center=<0.1 0 0>, rotation=<rot 0 0 90>}]
 dyn_headCane_02 [{shapeType=Capsule, length=10.4698, radius=10, center=<0.1 0 0>, rotation=<rot 0 0 90>}]
 dyn_headCane_03 [{shapeType=Capsule, length=8.1831, radius=10, center=<0.1 0 0>, rotation=<rot 0 0 90>}]
 dyn_headCane_04 [{shapeType=Capsule, length=7.0871, radius=10, center=<0.1 0 0>, rotation=<rot 0 0 90>}]
 dyn_headCane_05 [{shapeType=Capsule, length=5.9776, radius=10, center=<0.1 0 0>, rotation=<rot 0 0 90>}]
 dyn_headCane_mstr [{shapeType=Capsule, length=0.505, radius=0.505, rotation=<rot 0 -90 0>}]
}
constraints [
 {bone1Name=dyn_headCane_02, bone2Name=dyn_headCane_01, swing1LimitDegrees=2, swing2LimitDegrees=2, twistLimitDegrees=2}
 {bone1Name=dyn_headCane_04, bone2Name=dyn_headCane_03, swing1LimitDegrees=2, swing2LimitDegrees=2, twistLimitDegrees=2}
 {bone1Name=dyn_headCane_03, bone2Name=dyn_headCane_02, swing1LimitDegrees=2, swing2LimitDegrees=2, twistLimitDegrees=2}
 {bone1Name=dyn_headCane_05, bone2Name=dyn_headCane_04, swing1LimitDegrees=2, swing2LimitDegrees=2, twistLimitDegrees=2}
 {bone1Name=dyn_headCane_01, bone2Name=dyn_headCane_mstr, swing1LimitDegrees=2, swing2LimitDegrees=2, twistLimitDegrees=2}
]
